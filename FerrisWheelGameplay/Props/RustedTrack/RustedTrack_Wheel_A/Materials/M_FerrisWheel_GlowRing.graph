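# Translucent, unlit, two-sided glow ring material for the Ferris wheel, with spline-thickened geometry and a camera offset.
# Pivot-relative world position becomes angle and distance masks that drive animated hue-shifted LEDs, expanding pulses and spinning segments.
# Camera distance blends side LEDs with the wheel center pattern.
# Opacity combines a soft ring with scene-depth-based fake light, faded by depth, camera distance and a camera-direction mask.
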Material /FerrisWheelGameplay/Props/RustedTrack/RustedTrack_Wheel_A/Materials/M_FerrisWheel_GlowRing
graph {
 properties {bAutomaticallySetUsageInEditor=false, blendMode=BLEND_Translucent, shadingModel=MSM_Unlit, twoSided=true, usedShadingModels=Unlit}
 nodes [
  n1 = ScalarParameter() {defaultValue=1600, parameterName=RingWidth}
  Add(
   A=MaterialFunctionCall(WidthBase=@n1, WidthTip=@n1) {materialFunction=&SplineThicken}.WorldPositionOffset
   B=MaterialFunctionCall(Offset Amount=ScalarParameter() {defaultValue=500, parameterName=CameraOffset}) {materialFunction=/Engine/Functions/Engine_MaterialFunctions02/WorldPositionOffset/CameraOffset}.Camera Offset
  )
  Add()
  n2 = Constant3Vector() {constant=<rgba 0 0 1 0>}
  n3 = ScalarParameter() {parameterName=ActorRotationforUV}
  n4 = MaterialFunctionCall() {materialFunction=&ObjectPivotPoint}
  n5 = WorldPosition()
  n6 = MaterialFunctionCall(Float3=Add(A=RotateAboutAxis(NormalizedRotationAxis=@n2, RotationAngle=@n3, PivotPoint=@n4.Object Pivot Location, Position=@n5.XYZ), B=@n5.XYZ)) {materialFunction=&BreakOutFloat3Components}
  n7 = AppendVector(A=@n6.R, B=@n6.R)
  n8 = MaterialFunctionCall(Float3=@n4.Object Pivot Location) {materialFunction=&BreakOutFloat3Components}
  n9 = AppendVector(A=@n8.R, B=@n8.R)
  n10 = MaterialFunctionCall(Vector or UVs=Divide(A=Subtract(A=@n7, B=@n9), B=ScalarParameter() {defaultValue=4000, parameterName=FerrisWheelScale})) {materialFunction=&VectorToRadialValue}
  AppendVector(
   A=NamedRerouteDeclaration(None=@n10.Vector Converted to Angle) {name=AngleMask, nodeColor=<rgba 1 0.8203 0 1>}
   B=Multiply(
    A=NamedRerouteDeclaration(None=@n10.Linear Distance) {name=DistanceMask, nodeColor=<rgba 1 0 0.3281 1>}
    B=ScalarParameter() {defaultValue=2, parameterName=LEDTiling_V}
   )
  )
  Constant() {r=1}
  Constant()
  n11 = ComponentMask(None=TextureCoordinate()) {r=true}
  n12 = Add(
   A=SphereMask(A=@n11, B=Constant() {r=0.9}) {attenuationRadius=0.016, hardnessPercent=0}
   B=SphereMask(A=@n11, B=ScalarParameter() {defaultValue=0.82, parameterName=Param_1}) {attenuationRadius=0.016, hardnessPercent=0}
  )
  n13 = Time()
  n14 = ScalarParameter() {defaultValue=0.25, parameterName=SideLEDBreathingSpeed}
  n15 = Multiply(
   A=@n12
   B=MaterialFunctionCall(
    Hue Shift Percentage=Multiply(A=@n13, B=@n14)
    Texture=VectorParameter() {parameterName=LEDSideColor, defaultValue=<rgba 0.6667 0.0083 0 1>}.RGB
   ) {materialFunction=&HueShift}.Result
  )
  n16 = NamedRerouteUsage()
  n17 = Frac(None=Add(A=@n16) {constB=0.25})
  n18 = Constant() {r=0.5}
  n19 = Time()
  n20 = ScalarParameter() {defaultValue=0.04, parameterName=WheelEmissiveExpandingSpeed_C}
  n21 = Multiply(A=Frac(None=Multiply(A=@n19, B=@n20)), B=ScalarParameter() {defaultValue=4, parameterName=WheelEmissiveExpandingDuration_C}) {constB=3}
  n22 = Saturate(None=@n21)
  n23 = ScalarParameter() {parameterName=WheelEmissiveExpandingHardness_C}
  n24 = Multiply(A=NamedRerouteUsage(), B=OneMinus(None=SmoothStep(Value=@n21) {constMax=4, constMin=3.5}))
  n25 = SmoothStep(Value=@n22) {constMax=0.8, constMin=0.6}
  n26 = Multiply(
   A=MaterialFunctionCall(Hue Shift Percentage=@n16, Texture=Constant3Vector() {constant=<rgba 0 0 1 0>}) {materialFunction=&HueShift}.Result
   B=Multiply(A=SphereMask(A=@n17, B=@n18, Radius=@n22, Hardness=@n23) {hardnessPercent=0.5}, B=LinearInterpolate(B=@n24, Alpha=@n25) {constA=1})
  )
  n27 = Step(X=@n22) {constY=0.8}
  n28 = Add(A=NamedRerouteUsage()) {constB=0.25}
  n29 = Frac(None=Multiply(A=Time(), B=ScalarParameter() {defaultValue=0.2, parameterName=WheelEmissiveMovingSpeed_C2}))
  n30 = ScalarParameter() {defaultValue=0.9, parameterName=WheelEmissiveMovingSpeedPow_C2}
  n31 = SphereMask(
   A=Distance(A=Frac(None=@n28), B=Constant() {r=0.5})
   B=Multiply(A=Power(Base=@n29, Exp=@n30)) {constB=0.7}
   Radius=ScalarParameter() {defaultValue=0.05, parameterName=WheelEmissiveRadius_C2}
   Hardness=ScalarParameter() {parameterName=WheelEmissiveExpandingHardness_C2}
  ) {hardnessPercent=0.5}
  n32 = Multiply(A=Time(), B=ScalarParameter() {defaultValue=0.15, parameterName=WheelEmissiveMovingHueShiftSpeed_C2})
  n33 = VectorParameter() {parameterName=WheelEmissiveRotatingColorC2, defaultValue=<rgba 0.5833 0.0169 0 1>}
  n34 = Multiply(A=Multiply(
    A=Multiply(A=@n27, B=@n31)
    B=MaterialFunctionCall(Hue Shift Percentage=@n32, Texture=@n33.RGB) {materialFunction=&HueShift}.Result
   )) {constB=0.5}
  n35 = Multiply(B=Multiply(A=Add(A=@n26, B=@n34), B=ScalarParameter() {defaultValue=5, parameterName=WheelEmissiveMult_C})) {constA=1}
  n36 = Add(A=@n35, B=Multiply(A=@n35, B=NamedRerouteUsage()))
  n37 = NamedRerouteUsage()
  n38 = MaterialFunctionCall!(Distance=ScalarParameter() {defaultValue=5000, parameterName=WheelCenterBeamTransitionDistance}) {materialFunction=/Game/Effects/Fort_Effects/Materials/_Functions/MF_CameraDistanceFade}
  n39 = SmoothStep(Value=@n38.Result) {constMin=0.3}
  n40 = NamedRerouteDeclaration(None=LinearInterpolate(A=LinearInterpolate(A=@n15, B=@n36, Alpha=@n37), B=@n36, Alpha=Multiply(A=@n39))) {name=WheeleEmissive, nodeColor=<rgba 0 0.625 1 1>}
  Desaturation(None=Saturate(None=@n40), Fraction=Constant() {r=1})
  LinearInterpolate(A=@n12)
  n41 = SphereMask(A=@n40, B=Constant()) {attenuationRadius=0.5, hardnessPercent=0}
  n42 = ComponentMask(None=TextureCoordinate()) {r=true}
  n43 = Constant() {r=0.5}
  n44 = ScalarParameter() {defaultValue=0.6, parameterName=RingThickness}
  n45 = Multiply(
   A=Power(Base=OneMinus(None=@n41)) {constExponent=0.5}
   B=Power(
    Base=SphereMask(A=@n42, B=@n43, Radius=@n44) {attenuationRadius=0.5, hardnessPercent=0}
    Exp=ScalarParameter() {defaultValue=3, parameterName=RingOpacityPower}
   )
  )
  n46 = ScalarParameter() {defaultValue=0.05, parameterName=RingOpacity}
  n47 = Divide(A=Subtract(A=SceneDepth().R, B=PixelDepth().R), B=ScalarParameter() {defaultValue=150, parameterName=LightAttenuation})
  n48 = Saturate(None=DepthFade(Opacity=OneMinus(None=@n47), FadeDistance=ScalarParameter() {defaultValue=500, parameterName=DepthFadeDistance}))
  n49 = ScalarParameter() {defaultValue=0.5, parameterName=FakeLightOpacity}
  n50 = DepthFade(
   Opacity=Add(A=Multiply(A=@n45, B=@n46), B=Multiply(A=@n48, B=@n49))
   FadeDistance=ScalarParameter() {defaultValue=500, parameterName=DepthFadeDistance}
  )
  n51 = MaterialFunctionCall(
   Fade Length=ScalarParameter() {defaultValue=5000, parameterName=FadeLength}
   Fade Offset=ScalarParameter() {defaultValue=5000, parameterName=FadeOffset}
  ) {materialFunction=&CameraDepthFade}
  n52 = Normalize(VectorInput=TransformPosition(None=CameraPositionWS()) {transformSourceType=TRANSFORMPOSSOURCE_World})
  n53 = OneMinus(None=SphereMask(A=ComponentMask(None=@n52) {b=true}, B=Constant(), Radius=ScalarParameter() {defaultValue=1, parameterName=FadingDistacnce}) {attenuationRadius=200, hardnessPercent=0})
  n54 = ScalarParameter() {defaultValue=2, parameterName=FadingPower}
  NamedRerouteDeclaration(None=LinearInterpolate(B=Multiply(A=@n50, B=@n51.Result), Alpha=Power(Base=@n53, Exp=@n54) {constExponent=50}) {constB=0}) {name=Opacity, nodeColor=<rgba 0 0.2734 1 1>}
  n55 = Add(A=NamedRerouteUsage(), B=Multiply(A=Time(), B=ScalarParameter() {defaultValue=-0.05, parameterName=WheelSegmentSpinningSpeed_A}))
  n56 = ScalarParameter() {defaultValue=14, parameterName=WheelSegmentMult_A}
  n57 = SphereMask(
   A=Frac(None=Multiply(A=@n55, B=@n56))
   B=Constant() {r=0.5}
   Radius=ScalarParameter() {defaultValue=1, parameterName=WheelSegmentWidth_A}
   Hardness=ScalarParameter() {parameterName=WheelSegmentHardness_A}
  ) {attenuationRadius=0.5, hardnessPercent=0}
  NamedRerouteDeclaration(None=@n57) {name=WheelSegmentMask_A, nodeColor=<rgba 0 0.6484 1 1>}
  NamedRerouteDeclaration(None=@n38.Result) {name=Transition Distance, nodeColor=<rgba 0.2891 0 1 1>}
  NamedRerouteDeclaration(None=@n10.Radial Coordinates) {name=RadialUV, nodeColor=<rgba 0 0.8125 1 1>}
  NamedRerouteDeclaration(None=@n40) {name=Debug, nodeColor=<rgba 1 0.375 0 1>}
  NamedRerouteUsage()
  NamedRerouteUsage()
  NamedRerouteUsage()
  n58 = Multiply(
   A=LinearInterpolate(
    A=VectorParameter() {parameterName=WheelEmissiveRotatingColorA1, defaultValue=<rgba 0.0025 0 1 1>}.RGB
    B=VectorParameter() {parameterName=WheelEmissiveRotatingColorA2, defaultValue=<rgba 0 0 0 1>}.RGB
    Alpha=@n57
   )
   B=ScalarParameter() {defaultValue=10, parameterName=WheelEmissiveMult_A}
  )
  n59 = VectorParameter() {parameterName=WheelEmissiveRotatingColorB, defaultValue=<rgba 0 0.8125 1 1>}
  n60 = ScalarParameter() {defaultValue=20, parameterName=WheelEmissiveMult_B}
  n61 = Add(A=NamedRerouteUsage(), B=Multiply(A=Time(), B=ScalarParameter() {defaultValue=0.1, parameterName=WheelSegmentSpinningSpeed_B}))
  n62 = ScalarParameter() {defaultValue=8, parameterName=WheelSegmentMult_B}
  n63 = SphereMask(
   A=Frac(None=Multiply(A=@n61, B=@n62))
   B=Constant() {r=0.5}
   Radius=ScalarParameter() {defaultValue=0.2, parameterName=WheelSegmentWidth_B}
   Hardness=ScalarParameter() {parameterName=WheelSegmentHardnessB}
  ) {attenuationRadius=0.5, hardnessPercent=0}
  n64 = Power(Base=MaterialFunctionCall(Speed=ScalarParameter() {defaultValue=0.6, parameterName=WheelEmissiveBlinkingSpeed}) {materialFunction=&DebugTimeSine}.Result) {constExponent=1.5}
  Reroute(None=LinearInterpolate(A=@n58, B=Multiply(A=@n59.RGB, B=@n60), Alpha=Multiply(A=@n63, B=@n64)))
  Reroute(None=@n58)
  SphereMask(A=ComponentMask(None=TextureCoordinate()) {g=true}, B=Constant() {r=0.984}) {attenuationRadius=0.5, hardnessPercent=0.9}
  WorldPosition()
 ]
}
