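# MF_GreaterThan is a scalar material function in the NeonCity screen techart set that tests whether input X is greater than input Y.
# It subtracts Y from X, takes the sign of the difference and clamps it with Max against zero.
# This branchless step outputs 1 when X exceeds Y and 0 otherwise, including when they are equal.
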MaterialFunction /Game/Environments/Asteria/Sets/NeonCity/Techart/Screen/MaterialFunctions/Conditionals/MF_GreaterThan
graph {nodes [
  n1 = Subtract(
   A=FunctionInput() {inputName=X, inputType=FunctionInput_Scalar}
   B=FunctionInput() {inputName=Y, inputType=FunctionInput_Scalar, sortPriority=1}
  )
  FunctionOutput(None=Max(A=Sign(None=@n1)) {constB=0})
 ]}
