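# Material function in the RustedTrack wheel prop folder that maps UVs onto a grid of cells for a pixelated look.
# It centers input UVs, rescales them by object scale, cell size and an axis-dependent lerp, then floors them with pixel-aspect correction to snap to cells.
# It outputs tiling, pixelated and unpixelated UVs, clamped and unclamped, plus DDX/DDY.
# A reflection-vector bump offset is built but feeds no output.
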MaterialFunction /FerrisWheelGameplay/Props/RustedTrack/RustedTrack_Wheel_A/Materials/MF_CellMapping
graph {nodes [
  Divide(A=Constant() {r=1}, B=NamedRerouteUsage())
  n1 = MaterialFunctionCall() {materialFunction=&ObjectScale}
  n2 = Divide(A=@n1.Scale X, B=@n1.Scale X)
  n3 = Constant()
  n4 = Divide(A=@n1.Scale Z, B=@n1.Scale Z)
  n5 = MaterialFunctionCall!(X=@n4, Y=@n2) {materialFunction=/Game/Environments/Asteria/Sets/NeonCity/Techart/Screen/MaterialFunctions/Conditionals/MF_GreaterThan}
  n6 = LinearInterpolate(
   A=MaterialFunctionCall(X=OneMinus(None=@n2), Y=@n3) {materialFunction=&MakeFloat2}.Result
   B=MaterialFunctionCall(X=@n3, Y=OneMinus(None=@n4)) {materialFunction=&MakeFloat2}.Result
   Alpha=@n5.Result
  )
  n7 = Divide(A=@n6)
  Divide(A=@n7)
  n8 = Subtract(A=FunctionInput(Preview=TextureCoordinate()) {inputName=TextureCoordinate, inputType=FunctionInput_Vector2}) {constB=0.5}
  n9 = NamedRerouteUsage()
  n10 = Constant() {r=1}
  n11 = MaterialFunctionCall(X=@n2, Y=@n10) {materialFunction=&MakeFloat2}
  n12 = MaterialFunctionCall(X=@n10, Y=@n4) {materialFunction=&MakeFloat2}
  n13 = Constant() {r=0.05}
  n14 = Divide(
   A=LinearInterpolate(A=@n1.Scale Z, B=@n1.Scale Z, Alpha=@n5.Result)
   B=Divide(A=FunctionInput(Preview=@n13) {bUsePreviewValueAsDefault=true, inputName=cellSize, inputType=FunctionInput_Scalar}, B=NamedRerouteUsage())
  )
  n15 = Multiply(A=Multiply(A=Divide(A=@n8, B=@n9), B=LinearInterpolate(A=@n11.Result, B=@n12.Result, Alpha=@n5.Result)), B=@n14)
  n16 = Divide(B=FunctionInput(Preview=Constant() {r=1}) {bUsePreviewValueAsDefault=true, inputName=PixelAspect, inputType=FunctionInput_Scalar})
  n17 = Constant() {r=1}
  n18 = Floor(None=Multiply(A=@n15, B=AppendVector(A=@n16, B=@n17)))
  n19 = Subtract(A=Add(A=Divide(A=@n18, B=@n14)) {constB=0.5}, B=NamedRerouteUsage())
  FunctionOutput(None=Clamp(None=@n19)) {outputName=clamped_uvs}
  FunctionOutput(None=@n15) {outputName=tilingUVs}
  FunctionOutput(None=DDX()) {outputName=DDX}
  FunctionOutput(None=DDY()) {outputName=DDY}
  n20 = Subtract(A=Add(A=Divide(A=@n15, B=@n14)) {constB=0.5}, B=NamedRerouteUsage())
  FunctionOutput(None=Clamp(None=@n20)) {outputName=UnPixelatedUVs}
  FunctionOutput(None=@n19) {outputName=unclamped_uvs}
  FunctionOutput(None=@n20) {outputName=UnPixelatedUVNoClamp}
  n21 = MaterialFunctionCall!(
   Normal=Constant3Vector() {constant=<rgba 0 0 1 0>}
   CameraVector=Transform(None=CameraVectorWS()) {transformSourceType=TRANSFORMSOURCE_World, transformType=TRANSFORM_Tangent}
  ) {materialFunction=/Engine/Functions/Engine_MaterialFunctions02/Utility/CustomReflectionVector}
  Multiply(
   A=Divide(A=ComponentMask(None=@n21.Reflection Vector) {g=true, r=true}, B=ComponentMask(None=@n21.Reflection Vector) {b=true})
   B=FunctionInput(Preview=Constant()) {bUsePreviewValueAsDefault=true, inputName=bump, inputType=FunctionInput_Scalar}
  )
  n22 = FunctionInput(Preview=Constant() {r=1}) {bUsePreviewValueAsDefault=true, inputName=offsetToEdge, inputType=FunctionInput_Scalar, previewValue=<-1 0 0 1>}
  Multiply(A=@n22) {constB=-1}
  NamedRerouteDeclaration(None=@n7) {name=offset, nodeColor=<rgba 0 1 0.2734 1>}
  NamedRerouteDeclaration(None=Multiply(A=NamedRerouteUsage(), B=@n22)) {name=toEdge, nodeColor=<rgba 0.3594 0 1 1>}
  NamedRerouteDeclaration(None=FunctionInput(Preview=Constant() {r=1}) {bUsePreviewValueAsDefault=true, inputName=uvscale, inputType=FunctionInput_Vector2}) {name=uvscale, nodeColor=<rgba 0 0.0391 1 1>}
  OneMinus(None=@n6)
  VertexNormalWS()
 ]}
