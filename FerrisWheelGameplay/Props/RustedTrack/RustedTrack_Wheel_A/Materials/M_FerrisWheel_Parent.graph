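# Parent material for the Ferris wheel prop that draws its animated LED light show on the wheel and pillars, selected by an IsWheel static switch.
# Pivot-relative world position becomes radial UVs for hue-shifting, spinning and expanding patterns, while an FX mask separates small lights, side panels and center beam.
# Tiled LED textures add pixel detail and height-derived normals, and side-panel detail blends by camera distance.
# Emissive scales with time of day, ray-tracing quality and inverse eye adaptation.
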Material /FerrisWheelGameplay/Props/RustedTrack/RustedTrack_Wheel_A/Materials/M_FerrisWheel_Parent
graph {
 properties {
  bAlwaysEvaluateWorldPositionOffset=true
  bAutomaticallySetUsageInEditor=false
  bTangentSpaceNormal=false
  bUsedWithInstancedStaticMeshes=true
  bUsedWithNanite=true
  usedShadingModels=Default Lit
 }
 nodes [
  n1 = OneMinus(None=NamedRerouteUsage())
  n2 = Multiply(A=@n1, B=@n1)
  n3 = NamedRerouteUsage()
  n4 = LinearInterpolate(A=@n3) {constB=0.1}
  Add(A=Multiply(A=Multiply(A=@n2, B=@n4)) {constB=0.2}, B=@n3)
  Add(A=NamedRerouteUsage(), B=Multiply(A=NamedRerouteUsage()) {constB=0.5})
  n5 = AppendVector(A=Constant() {r=-7.3138}, B=Constant() {r=-0.2894})
  n6 = Distance(A=Add(A=TextureCoordinate(), B=Multiply(A=@n5) {constB=0.1}) {constB=-0.5}, B=Constant())
  n7 = Multiply(A=@n6, B=ScalarParameter() {defaultValue=622.6183, parameterName=Circle Amount})
  n8 = Multiply(A=Add(A=Sine(None=@n7)) {constB=1.5}) {constB=0.5}
  n9 = OneMinus(None=Multiply(A=@n6, B=ScalarParameter() {defaultValue=60.9877, parameterName=Param_4}))
  n10 = AppendVector(A=ScalarParameter() {defaultValue=-3.4894, parameterName=Param_5}, B=ScalarParameter() {defaultValue=-9.2939, parameterName=Param_6})
  n11 = Distance(A=Add(A=TextureCoordinate(), B=Multiply(A=@n10) {constB=0.1}) {constB=-0.5}, B=Constant())
  n12 = Sine(None=Multiply(A=@n11, B=ScalarParameter() {defaultValue=400, parameterName=Circle Amount 2}))
  n13 = ScalarParameter() {defaultValue=39.1644, parameterName=Param_7}
  n14 = Multiply(A=Multiply(A=Add(A=@n12) {constB=1.5}) {constB=0.5}, B=OneMinus(None=Multiply(A=@n11, B=@n13)))
  n15 = AppendVector(A=ScalarParameter() {defaultValue=-3.545, parameterName=Param_8}, B=ScalarParameter() {defaultValue=-8.5714, parameterName=Param_9})
  n16 = Distance(A=Add(A=TextureCoordinate(), B=Multiply(A=@n15) {constB=0.1}) {constB=-0.5}, B=Constant())
  n17 = Sine(None=Multiply(A=@n16, B=ScalarParameter() {defaultValue=400, parameterName=Circle Amount 2}))
  n18 = ScalarParameter() {defaultValue=39.1644, parameterName=Param_7}
  n19 = Multiply(A=Multiply(A=Add(A=@n17) {constB=1.5}) {constB=0.5}, B=OneMinus(None=Multiply(A=@n16, B=@n18)))
  n20 = Add(A=Saturate(None=Multiply(A=@n8, B=@n9)), B=Add(A=Saturate(None=@n14), B=Saturate(None=@n19)))
  Add(A=@n20, B=NamedRerouteUsage())
  n21 = Constant3Vector() {constant=<rgba 0 0 1 0>}
  n22 = ScalarParameter() {parameterName=ActorRotationforUV}
  n23 = MaterialFunctionCall() {materialFunction=&ObjectPivotPoint}
  n24 = WorldPosition()
  n25 = MaterialFunctionCall(Float3=Add(A=RotateAboutAxis(NormalizedRotationAxis=@n21, RotationAngle=@n22, PivotPoint=@n23.Object Pivot Location, Position=@n24.XYZ), B=@n24.XYZ)) {materialFunction=&BreakOutFloat3Components}
  n26 = AppendVector(A=@n25.R, B=@n25.R)
  n27 = MaterialFunctionCall(Float3=@n23.Object Pivot Location) {materialFunction=&BreakOutFloat3Components}
  n28 = AppendVector(A=@n27.R, B=@n27.R)
  n29 = MaterialFunctionCall(Vector or UVs=Divide(A=Subtract(A=@n26, B=@n28), B=ScalarParameter() {defaultValue=4000, parameterName=FerrisWheelScale})) {materialFunction=&VectorToRadialValue}
  AppendVector(
   A=NamedRerouteDeclaration(None=@n29.Vector Converted to Angle) {name=AngleMask, nodeColor=<rgba 1 0.8203 0 1>}
   B=Multiply(
    A=NamedRerouteDeclaration(None=@n29.Linear Distance) {name=DistanceMask, nodeColor=<rgba 1 0 0.3281 1>}
    B=ScalarParameter() {defaultValue=2, parameterName=LEDTiling_V}
   )
  )
  n30 = ComponentMask(None=TextureCoordinate()) {r=true}
  n31 = Add(
   A=SphereMask(A=@n30, B=Constant() {r=0.9}) {attenuationRadius=0.016, hardnessPercent=0}
   B=SphereMask(A=@n30, B=ScalarParameter() {defaultValue=0.82, parameterName=Param_1}) {attenuationRadius=0.016, hardnessPercent=0}
  )
  LinearInterpolate(
   A=@n31
   B=TextureSampleParameter2D() {parameterName=FXMask, texture=/FerrisWheelGameplay/Props/RustedTrack/RustedTrack_Wheel_A/Textures/T_RustedTrack_WheelRing_A_A_M}.G
  )
  LinearInterpolate()
  n32 = TextureCoordinate()
  n33 = VectorParameter() {parameterName=CellUVScale_CenterBeam, defaultValue=<rgba 100 300 0 1>}
  Multiply(A=@n32, B=AppendVector(A=@n33.R, B=@n33.R))
  n34 = MaterialFunctionCall(
   Hue Shift Percentage=Multiply(A=Time(), B=ScalarParameter() {defaultValue=0.3, parameterName=SmallLightHueShiftSpeed})
   Texture=VectorParameter() {parameterName=SmallLightColor1, defaultValue=<rgba 1 0 0.0548 1>}.RGB
  ) {materialFunction=&HueShift}
  n35 = NamedRerouteUsage()
  n36 = Frac(None=Multiply(A=ComponentMask(None=@n35) {r=true}, B=ScalarParameter() {defaultValue=61}) {constB=50})
  n37 = MaterialFunctionCall(Speed=ScalarParameter() {defaultValue=0.5, parameterName=SmallLightFlickeringSpeed}) {materialFunction=&DebugTimeSine}
  Multiply(A=@n34.Result, B=LinearInterpolate(A=Step(Y=@n36) {constX=0.5}, B=Step(X=@n36) {constX=0.5, constY=0.5}, Alpha=Power(Base=@n37.Result)))
  n38 = TextureCoordinate()
  n39 = MaterialFunctionCall!(
   bump
   offsetToEdge=ScalarParameter() {parameterName=EdgeOffset}
   uvscale
   PixelAspect
   TextureCoordinate=AppendVector(A=ComponentMask(None=@n38) {g=true}, B=ComponentMask(None=@n38) {r=true})
   cellSize=LinearInterpolate(
    A=ScalarParameter() {defaultValue=0.0145, parameterName=CellSize_SidePanel}
    B=ScalarParameter() {defaultValue=0.003, parameterName=CellSize_CenterBeam}
    Alpha=NamedRerouteUsage()
   )
  ) {materialFunction=&/MF_CellMapping}
  n40 = NamedRerouteDeclaration(None=@n39.tilingUVs) {name=ScaledUVs, nodeColor=<rgba 1 0 0.0938 1>}
  n41 = ComponentMask(None=@n40) {r=true}
  n42 = Constant() {r=8}
  n43 = AppendVector(
   A=Constant()
   B=Subtract(A=MaterialFunctionCall!(PrimaryTextureCoordinates=@n41, 2ndOrderTextureCoordinates, RandomSeed=@n42, NoiseDensityBias, Bypass Noise Influence Crosstalk) {materialFunction=/Game/Effects/Fort_Effects/Materials/_Functions/MF_RandomUVSample}.NoiseWithRange) {constB=0.5}
  )
  Multiply(A=@n43)
  n44 = VectorParameter() {parameterName=SmallLightColor1, defaultValue=<rgba 1 0 0.0548 1>}
  n45 = MaterialFunctionCall(
   Hue Shift Percentage=Multiply(A=Time(), B=ScalarParameter() {defaultValue=0.7, parameterName=PillarSmallLightHueShiftSpeed})
   Texture=VectorParameter() {parameterName=PillarSmallLightColor2, defaultValue=<rgba 5 3 0 1>}.RGB
  ) {materialFunction=&HueShift}
  n46 = Time()
  n47 = ScalarParameter() {defaultValue=0.7, parameterName=PillarSmallLightSpeed}
  n48 = SphereMask(
   A=Multiply(A=NamedRerouteUsage(), B=ScalarParameter() {defaultValue=1, parameterName=Param2}) {constB=50}
   B=Frac(None=Multiply(A=@n46, B=@n47))
  ) {attenuationRadius=0.1, hardnessPercent=0}
  NamedRerouteDeclaration(None=Multiply(A=LinearInterpolate(A=@n44.RGB, B=@n45.Result, Alpha=@n48), B=ScalarParameter() {defaultValue=5, parameterName=PillarSmallLightEmissiveMult})) {name=SmallLightEmissive_Pillar, nodeColor=<rgba 0 0.8125 1 1>}
  NamedRerouteDeclaration(None=Multiply(A=@n34.Result, B=ScalarParameter() {defaultValue=5, parameterName=SmallLightEmissiveMult})) {name=SmallLightWheel, nodeColor=<rgba 1 0.8672 0 1>}
  n49 = LinearInterpolate(B=NamedRerouteUsage(), Alpha=NamedRerouteUsage()) {constB=0}
  n50 = ScalarParameter() {defaultValue=0.1, parameterName=LED_BG_BaseColorStrength}
  NamedRerouteDeclaration(None=LinearInterpolate(
    A=Multiply(A=@n49, B=@n50)
    B=VectorParameter() {parameterName=SmallLightBaseColor, defaultValue=<rgba 1 0.4637 0 1>}.RGB
    Alpha=NamedRerouteUsage()
   )) {name=BaseColor, nodeColor=<rgba 0.5469 0 1 1>}
  n51 = NamedRerouteUsage()
  n52 = NamedRerouteUsage()
  n53 = NamedRerouteUsage()
  n54 = Add(A=NamedRerouteUsage()) {constB=0.1}
  n55 = Multiply(A=@n54, B=NamedRerouteUsage())
  n56 = NamedRerouteUsage()
  n57 = Multiply(
   A=StaticSwitchParameter(True=LinearInterpolate(A=@n51, B=@n52, Alpha=@n53), False=NamedRerouteUsage()) {defaultValue=true, parameterName=IsWheel}
   B=LinearInterpolate(A=LinearInterpolate(A=@n54, B=@n55, Alpha=@n56), Alpha=NamedRerouteUsage())
  )
  n58 = Add(A=NamedRerouteUsage()) {constB=0.05}
  n59 = NamedRerouteUsage()
  n60 = StaticSwitchParameter(True=LinearInterpolate(A=LinearInterpolate(A=@n58, Alpha=@n59), Alpha=NamedRerouteUsage()), False=Constant() {r=1}) {defaultValue=true, parameterName=IsWheel}
  n61 = Multiply(
   A=Multiply(A=Multiply(A=@n57, B=@n60), B=NamedRerouteUsage())
   B=MaterialFunctionCall(
    Value_Morning=ScalarParameter() {defaultValue=0.65, group=Global Emissive Adjustments, parameterName=Emissive_Morning}
    Value_Day=ScalarParameter() {defaultValue=0.6, group=Global Emissive Adjustments, parameterName=Emissive_Day}
    Value_Evening=ScalarParameter() {defaultValue=0.65, group=Global Emissive Adjustments, parameterName=Emissive_Evening}
    Value_Night=Constant() {r=1}
   ) {materialFunction=&Scalar_Time_of_Day}.Result
  )
  n62 = RayTracingQualitySwitch(Normal=@n61, RayTraced=Multiply(A=@n61, B=ScalarParameter() {defaultValue=2, parameterName=LumenBoost}))
  n63 = StaticSwitchParameter(
   True=LinearInterpolate(
    A=ScalarParameter() {defaultValue=100, parameterName=Intensity Main Bean}
    B=ScalarParameter() {defaultValue=10, parameterName=Intensity Side Lights}
    Alpha=NamedRerouteUsage()
   ) {constA=100, constB=10}
   False=ScalarParameter() {defaultValue=30, parameterName=Intensity Front Pilar Lights}
  ) {defaultValue=true, parameterName=IsWheel}
  NamedRerouteDeclaration(None=EyeAdaptationInverse(LightValueInput=Multiply(A=@n62, B=@n63) {constB=3}, AlphaInput=ScalarParameter() {defaultValue=0.8, parameterName=EyeAdaptionInverse}).EyeAdaptationInverse) {name=Emissive, nodeColor=<rgba 0 1 0.2266 1>}
  n64 = Add(A=NamedRerouteUsage(), B=Multiply(A=Time(), B=ScalarParameter() {defaultValue=-0.05, parameterName=WheelSegmentSpinningSpeed_A}))
  n65 = ScalarParameter() {defaultValue=14, parameterName=WheelSegmentMult_A}
  n66 = SphereMask(
   A=Frac(None=Multiply(A=@n64, B=@n65))
   B=Constant() {r=0.5}
   Radius=ScalarParameter() {defaultValue=0.3, parameterName=WheelSegmentWidth_A}
   Hardness=ScalarParameter() {defaultValue=0.5, parameterName=WheelSegmentHardness_A}
  ) {attenuationRadius=0.5, hardnessPercent=0}
  NamedRerouteDeclaration(None=@n66) {name=WheelSegmentMask_A, nodeColor=<rgba 0 0.6484 1 1>}
  n67 = Multiply(A=ScalarParameter() {defaultValue=500, group=4-LCD, parameterName=LCDTiling}, B=Constant2Vector() {g=2, r=1})
  n68 = Multiply(A=@n32, B=@n67)
  n69 = TextureSampleParameter2D(UVs=BumpOffset(Coordinate=@n68, Height=ScalarParameter() {defaultValue=-5, parameterName=Param_2})) {
   parameterName=LED_BG_CenterBeam
   texture=/FerrisWheelGameplay/Props/RustedTrack/RustedTrack_Wheel_A/Textures/T_LCD_Panels_LCD_RustedTrackWheel
  }
  NamedRerouteDeclaration(None=NamedRerouteDeclaration(None=@n69.RGB) {name=LED_BG_CenterBeam, nodeColor=<rgba 0 1 0.8594 1>}) {name=Pixels, nodeColor=<rgba 1 0.7734 0 1>}
  n70 = TextureSampleParameter2D() {parameterName=FXMask, texture=&T_RustedTrack_WheelRing_A_A_M}
  n71 = NamedRerouteDeclaration(None=Saturate(None=Sign(None=@n70.R))) {name=SmallLightsMask, nodeColor=<rgba 1 0.5391 0 1>}
  n72 = NamedRerouteDeclaration(None=@n70.B) {name=CenterBeamMask, nodeColor=<rgba 0.2188 0 1 1>}
  NamedRerouteDeclaration(None=Subtract(A=Constant() {r=1}, B=Add(A=@n71, B=@n72))) {name=BGScreenMask, nodeColor=<rgba 0.1016 1 0 1>}
  n73 = VertexNormalWS()
  n74 = Add(A=@n69.R, B=@n69.R)
  n75 = Multiply(A=Divide(A=Add(A=@n74, B=@n69.B)) {constB=3}, B=NamedRerouteUsage())
  n76 = Multiply(A=MaterialFunctionCall!(
    Height=NamedRerouteDeclaration(None=@n75) {name=LED_Height, nodeColor=<rgba 1 0.1641 0 1>}
    Absolute World Position
    World Space Vertex Normals
   ) {materialFunction=/Engine/Functions/Engine_MaterialFunctions03/Procedurals/HeightToNormalSmooth}.WorldSpaceNormal)
  NamedRerouteDeclaration(None=LinearInterpolate(A=@n73, B=LinearInterpolate(A=@n76, B=@n73), Alpha=NamedRerouteUsage())) {name=WorldSpaceNormal, nodeColor=<rgba 0 0.1562 1 1>}
  n77 = ScalarParameter() {defaultValue=0.3, parameterName=Roughness_LED_High}
  n78 = ScalarParameter() {parameterName=Roughness_LED_Low}
  n79 = NamedRerouteUsage()
  NamedRerouteDeclaration(None=LinearInterpolate(A=ScalarParameter() {parameterName=Roughness}, B=LinearInterpolate(A=@n77, B=@n78, Alpha=@n79), Alpha=NamedRerouteUsage())) {name=Roughtness, nodeColor=<rgba 0.0781 0 1 1>}
  NamedRerouteDeclaration(None=Divide(A=Ceil(None=@n68), B=@n67)) {name=Pixelated UVs, nodeColor=<rgba 0.1953 1 0 1>}
  n80 = MaterialFunctionCall!(Distance=ScalarParameter() {defaultValue=10000, parameterName=WheelCenterBeamTransitionDistance}) {materialFunction=/Game/Effects/Fort_Effects/Materials/_Functions/MF_CameraDistanceFade}
  NamedRerouteDeclaration(None=@n80.Result) {name=Transition Distance, nodeColor=<rgba 0.2891 0 1 1>}
  n81 = TextureCoordinate()
  n82 = Sign(None=ComponentMask(None=Add(A=@n81) {constB=-0.89}) {g=true})
  n83 = Multiply(A=LinearInterpolate(Alpha=Saturate(None=@n82)) {constA=-1}, B=Time())
  NamedRerouteDeclaration(None=Saturate(None=Sine(None=@n83))) {name=Blinking Lights Mask, nodeColor=<rgba 0.1484 1 0 1>}
  n84 = MaterialFunctionCall!("Normal (WS)", Texture Scale=ScalarParameter() {defaultValue=92.8294, parameterName=Param_3}, Triplanar Fade, Derivatives Scale) {materialFunction=/Game/Athena/Environments/MaterialFunctions/MF_TriplanarDithered}
  n85 = VertexNormalWS()
  n86 = LinearInterpolate(
   B=TextureSample!(UVs=@n84.UVs, Tex, "DDX(UVs)"=@n84.UVs, "DDY(UVs)"=@n84.UVs, Apply View MipBias) {mipValueMode=TMVM_Derivative, texture=/Game/Packages/Fortress_Transylvania/Haunted_Forest/Swamps/Textures/clouds_dust_detail_01}.G
   Alpha=Saturate(None=ComponentMask(None=@n85) {b=true})
  ) {constA=1}
  n87 = Multiply(A=@n86, B=@n86)
  NamedRerouteDeclaration(None=Saturate(None=@n87)) {name=dust, nodeColor=<rgba 0.2891 1 0 1>}
  NamedRerouteDeclaration(None=@n87) {name=Base Dust, nodeColor=<rgba 1 0.9609 0 1>}
  NamedRerouteDeclaration(None=Multiply(A=@n70.R, B=@n70.R)) {name=Small Gradient, nodeColor=<rgba 1 0 0.2344 1>}
  NamedRerouteDeclaration(None=Add(A=@n20) {constB=0.1}) {name=Shape Circle, nodeColor=<rgba 0 0.3437 1 1>}
  NamedRerouteDeclaration(None=@n70.G) {name=Side Mask, nodeColor=<rgba 0.5469 1 0 1>}
  NamedRerouteDeclaration(None=@n29.Radial Coordinates) {name=RadialUV, nodeColor=<rgba 0 0.8125 1 1>}
  NamedRerouteDeclaration(None=Add(A=@n43, B=NamedRerouteDeclaration(None=@n39.clamped_uvs) {name=ClampedUVs, nodeColor=<rgba 0.8984 0 1 1>})) {name=QuantizedUVs, nodeColor=<rgba 1 0.8672 0 1>}
  NamedRerouteDeclaration(None=@n39.unclamped_uvs) {name=unclampedunpixel, nodeColor=<rgba 1 0.8438 0 1>}
  NamedRerouteDeclaration(None=Saturate(None=TextureSample!(UVs=@n40, Tex, Apply View MipBias) {texture=&T_LCD_Panels_LCD_RustedTrackWheel}.RGB)) {name=LED_BG, nodeColor=<rgba 0 1 0.8594 1>}
  n88 = Multiply(A=Time(), B=ScalarParameter() {defaultValue=0.25, parameterName=SideLEDBreathingSpeed})
  n89 = VectorParameter() {parameterName=LEDSideColor, defaultValue=<rgba 0.6667 0.0083 0 1>}
  n90 = NamedRerouteUsage()
  n91 = Frac(None=Add(A=@n90) {constB=0.25})
  n92 = Constant() {r=0.5}
  n93 = Time()
  n94 = ScalarParameter() {defaultValue=0.04, parameterName=WheelEmissiveExpandingSpeed_C}
  n95 = Multiply(A=Frac(None=Multiply(A=@n93, B=@n94)), B=ScalarParameter() {defaultValue=4, parameterName=WheelEmissiveExpandingDuration_C}) {constB=3}
  n96 = Saturate(None=@n95)
  n97 = ScalarParameter() {parameterName=WheelEmissiveExpandingHardness_C}
  n98 = Multiply(A=NamedRerouteUsage(), B=OneMinus(None=SmoothStep(Value=@n95) {constMax=4, constMin=3.5}))
  n99 = SmoothStep(Value=@n96) {constMax=0.8, constMin=0.6}
  n100 = Multiply(
   A=MaterialFunctionCall(Hue Shift Percentage=@n90, Texture=Constant3Vector() {constant=<rgba 0 0 1 0>}) {materialFunction=&HueShift}.Result
   B=Multiply(A=SphereMask(A=@n91, B=@n92, Radius=@n96, Hardness=@n97) {hardnessPercent=0.5}, B=LinearInterpolate(B=@n98, Alpha=@n99) {constA=1})
  )
  n101 = NamedRerouteUsage()
  n102 = Distance(A=Frac(None=Add(A=@n101) {constB=0.25}), B=Constant() {r=0.5})
  n103 = Multiply(A=Time(), B=ScalarParameter() {defaultValue=0.2, parameterName=WheelEmissiveMovingSpeed_C2})
  n104 = Multiply(A=Power(Base=Frac(None=@n103), Exp=ScalarParameter() {defaultValue=0.9, parameterName=WheelEmissiveMovingSpeedPow_C2})) {constB=0.7}
  n105 = ScalarParameter() {defaultValue=0.05, parameterName=WheelEmissiveRadius_C2}
  n106 = ScalarParameter() {parameterName=WheelEmissiveExpandingHardness_C2}
  n107 = Time()
  n108 = ScalarParameter() {defaultValue=0.15, parameterName=WheelEmissiveMovingHueShiftSpeed_C2}
  n109 = Multiply(
   A=Multiply(A=Step(X=@n96) {constY=0.8}, B=SphereMask(A=@n102, B=@n104, Radius=@n105, Hardness=@n106) {hardnessPercent=0.5})
   B=MaterialFunctionCall(
    Hue Shift Percentage=Multiply(A=@n107, B=@n108)
    Texture=VectorParameter() {parameterName=WheelEmissiveRotatingColorC2, defaultValue=<rgba 0.5833 0.0169 0 1>}.RGB
   ) {materialFunction=&HueShift}.Result
  )
  n110 = Multiply(B=Multiply(A=Add(A=@n100, B=@n109), B=ScalarParameter() {defaultValue=5, parameterName=WheelEmissiveMult_C})) {constA=1}
  n111 = Add(A=@n110, B=Multiply(A=@n110, B=NamedRerouteUsage()))
  n112 = LinearInterpolate(
   A=Multiply(
    A=@n31
    B=MaterialFunctionCall(Hue Shift Percentage=@n88, Texture=@n89.RGB) {materialFunction=&HueShift}.Result
   )
   B=@n111
   Alpha=NamedRerouteUsage()
  )
  n113 = Multiply(A=SmoothStep(Value=@n80.Result) {constMin=0.3})
  NamedRerouteDeclaration(None=NamedRerouteDeclaration(None=LinearInterpolate(A=@n112, B=@n111, Alpha=@n113)) {name=WheeleEmissive, nodeColor=<rgba 0 0.625 1 1>}) {name=Debug, nodeColor=<rgba 1 0.375 0 1>}
  NamedRerouteUsage()
  NamedRerouteUsage()
  NamedRerouteUsage()
  NamedRerouteUsage()
  NamedRerouteUsage()
  n114 = Multiply(
   A=LinearInterpolate(
    A=VectorParameter() {parameterName=WheelEmissiveRotatingColorA1, defaultValue=<rgba 0.0025 0 1 1>}.RGB
    B=VectorParameter() {parameterName=WheelEmissiveRotatingColorA2, defaultValue=<rgba 0 0 0 1>}.RGB
    Alpha=@n66
   )
   B=ScalarParameter() {defaultValue=10, parameterName=WheelEmissiveMult_A}
  )
  n115 = VectorParameter() {parameterName=WheelEmissiveRotatingColorB, defaultValue=<rgba 0 0.8125 1 1>}
  n116 = ScalarParameter() {defaultValue=20, parameterName=WheelEmissiveMult_B}
  n117 = Add(A=NamedRerouteUsage(), B=Multiply(A=Time(), B=ScalarParameter() {defaultValue=0.1, parameterName=WheelSegmentSpinningSpeed_B}))
  n118 = ScalarParameter() {defaultValue=8, parameterName=WheelSegmentMult_B}
  n119 = SphereMask(
   A=Frac(None=Multiply(A=@n117, B=@n118))
   B=Constant() {r=0.5}
   Radius=ScalarParameter() {defaultValue=0.2, parameterName=WheelSegmentWidth_B}
   Hardness=ScalarParameter() {parameterName=WheelSegmentHardnessB}
  ) {attenuationRadius=0.5, hardnessPercent=0}
  n120 = Power(Base=MaterialFunctionCall(Speed=ScalarParameter() {defaultValue=0.6, parameterName=WheelEmissiveBlinkingSpeed}) {materialFunction=&DebugTimeSine}.Result) {constExponent=1.5}
  Reroute(None=LinearInterpolate(A=@n114, B=Multiply(A=@n115.RGB, B=@n116), Alpha=Multiply(A=@n119, B=@n120)))
  Reroute(None=@n114)
  ScalarParameter() {defaultValue=1, parameterName=Specular}
  SphereMask(A=ComponentMask(None=TextureCoordinate()) {g=true}, B=Constant() {r=0.984}) {attenuationRadius=0.5, hardnessPercent=0.9}
 ]
}
